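# Mono one-shot MetaSound source for the BinGrass boss's footstep sweetener layer, routed to the boss fight foley sound class.
# On play, a random get picks a wave from an array variable and feeds a wave player, which signals finish when it ends.
# A dedicated attenuation asset shapes falloff, and overridden concurrency caps simultaneous voices with a short steal release.
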MetaSoundSource /JethroLager/NPCs/Bosses/BinGrass/Audio/MSS_BinGrass_Footsteps_Sweetener
inputs [
 UE.Source.OnPlay {type=Trigger}
 Wave Asset {type=WaveAsset}
]
interfaces [UE.OutputFormat.Mono:1.0, UE.Source.OneShot:1.0, UE.Source:1.1]
outputs [
 "UE.OutputFormat.Mono.Audio:0" {type=Audio}
 UE.Source.OneShot.OnFinished {type=Trigger}
]
pages [Default {
  input_defaults {UE.Source.OnPlay=null, Wave Asset=null}
  nodes [
   InitVariable.WaveAsset()
   InitVariable.WaveAsset()
   ArrayRandomGet_0 = Array.Random Get:WaveAsset:Array(Next=^UE.Source.OnPlay, In Array=VariableAccessor.WaveAsset:Array(Variable=InitVariable.WaveAsset:Array().Variable).Value)
   WavePlayer_0 = UE.Wave Player:Mono(Play=@ArrayRandomGet_0.On Next, Wave Asset=@ArrayRandomGet_0.Value)
  ]
  links [{dst="@UE.Source.OneShot.OnFinished", src=WavePlayer_0.On Finished}, {dst="@UE.OutputFormat.Mono.Audio:0", src=WavePlayer_0.Out Mono}]
 }]
properties {
 assetImportData=&self:AssetImportData
 attenuationSettings=&/ATT_Fort_Bingrass_Sweetener_Footsteps
 bOverrideConcurrency=true
 maxDistance=3400
 sampleRate=48000
 soundClassObject=/Game/Sounds/FortSoundClassesAndModes/Fort_BossFight_SoundClass/Fort_BossFight_Foley
 volume=2
 concurrencyOverrides {maxCount=2, voiceStealReleaseTime=0.1}
}
