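# Sound cue for the alarm loop in the BinGrass boss fight, routed to the Fort_Events_Gameplay sound class.
# It uses a dedicated boss-fight alarm attenuation asset with a long maximum distance and a lowered volume.
# A concurrency override caps it at two simultaneous instances.
# The file shows no root or sound nodes, so neither the wave source nor any looping logic appears here.
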SoundCue /JethroLager/NPCs/Bosses/BinGrass/Audio/SC_Bingrass_BossFight_Alarm_Loop
root=null
nodes []
properties {
 attenuationSettings=&/ATT_Fort_Bingrass_BossFight_Alarm
 bOverrideConcurrency=true
 maxDistance=35000
 soundClassObject=&Fort_Events_Gameplay
 volumeMultiplier=0.6
 concurrencyOverrides {maxCount=2}
}
